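# First-person firing sound cue for the Bone Assault Rifle, using ranged-weapon fire class, first-person attenuation and gunfire concurrency.
# A mixer layers one of three random close fire shots with a slightly delayed chuff and a later random shell casing from six variants.
# Modulators vary each layer, with wide random volume on chuff and casing and raised random pitch on casings, so repeated shots differ.
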SoundCue /PrimalGameplay/Items/Guns/Bone/BoneAssaultRifle/Assault_Bone_Fire_1P_Cue
root=Mixer_2
nodes [
 {id=Delay_0, children [Modulator_5], properties {delayMax=0.1, delayMin=0.1}}
 {id=Delay_2, children [Modulator_2], properties {delayMax=0.06, delayMin=0.06}}
 {id=Mixer_2, children [Modulator_3,Delay_2,Delay_0], properties {inputVolume [0.65,0.5,0.4]}}
 {id=Modulator_2, children [WavePlayer_1], properties {volumeMax=1, volumeMin=0.15}}
 {id=Modulator_3, children [Random_1]}
 {id=Modulator_5, children [Random_2], properties {pitchMax=2, pitchMin=1, volumeMax=1, volumeMin=0}}
 {id=Random_1, children [WavePlayer_6,WavePlayer_7,WavePlayer_8], properties {weights [1,1,1]}}
 {
  id=Random_2
  children [WavePlayer_11, WavePlayer_22, WavePlayer_23, WavePlayer_24, WavePlayer_25, WavePlayer_27]
  properties {weights [1,1,1,1,1,1]}
 }
 {id=WavePlayer_1, properties {soundWaveAssetPtr=&/Assault_Bone_Fire_Close_Chuff_01}}
 {id=WavePlayer_11, properties {soundWaveAssetPtr=&/Assault_Bone_Fire_Close_Shell_Casing_01}}
 {id=WavePlayer_22, properties {soundWaveAssetPtr=&/Assault_Bone_Fire_Close_Shell_Casing_02}}
 {id=WavePlayer_23, properties {soundWaveAssetPtr=&/Assault_Bone_Fire_Close_Shell_Casing_03}}
 {id=WavePlayer_24, properties {soundWaveAssetPtr=&/Assault_Bone_Fire_Close_Shell_Casing_04}}
 {id=WavePlayer_25, properties {soundWaveAssetPtr=&/Assault_Bone_Fire_Close_Shell_Casing_05}}
 {id=WavePlayer_27, properties {soundWaveAssetPtr=&/Assault_Bone_Fire_Close_Shell_Casing_06}}
 {id=WavePlayer_6, properties {soundWaveAssetPtr=&/Assault_Bone_Fire_Close_01}}
 {id=WavePlayer_7, properties {soundWaveAssetPtr=&/Assault_Bone_Fire_Close_02}}
 {id=WavePlayer_8, properties {soundWaveAssetPtr=&/Assault_Bone_Fire_Close_03}}
]
properties {
 attenuationSettings=&Fort_Weapon_Projectile_FirstPerson_Athena_Att
 maxDistance=2097152
 soundClassObject=&Fort_WeaponRanged_Fire_1P_PSM
 volumeMultiplier=1
 attenuationOverrides {distanceAlgorithm=NaturalSound, falloffDistance=5000, attenuationShapeExtents {x=1000}}
 concurrencyOverrides {bLimitToOwner=true, maxCount=3, resolutionRule=StopOldest}
 concurrencySet [&Weapon_Gunfire_FirstPerson_Conc]
}
