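# Sound cue for the Bone Assault Rifle's gunfire, blending close, medium and distant layers through per-layer attenuation.
# The close layer mixes a random shot with a briefly delayed, volume-modulated chuff under a third-person fire sound class.
# Medium and distant layers each pick a random wave, pass through modulators, and share a distant-fire sound class.
# Local and global gunfire concurrency sets govern simultaneous playback.
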SoundCue /PrimalGameplay/Items/Guns/Bone/BoneAssaultRifle/Assault_Bone_Fire_Cue
root=Mixer_17
nodes [
 {
  id=Attenuation_0
  children [Modulator_0]
  properties {attenuationSettings=&Fort_Weapon_Projectile_Medium_Mono_Athena_Att}
 }
 {
  id=Attenuation_34
  children [Modulator_1]
  properties {attenuationSettings=&Fort_Weapon_Projectile_Distant_Mono_Athena_Att}
 }
 {id=Attenuation_49, children [Mixer_1], properties {attenuationSettings=&Fort_Weapon_Projectile_Close_Athena_Att}}
 {id=Delay_2, children [Modulator_4], properties {delayMax=0.06, delayMin=0.06}}
 {id=Mixer_1, children [Modulator_5,Delay_2], properties {inputVolume [0.95,0.8]}}
 {id=Mixer_17, children [SoundClass_32, SoundClass_0, SoundClass_35], properties {inputVolume [0.7,0.7,0.5]}}
 {id=Modulator_0, children [Random_6], properties {pitchMax=1, volumeMax=1, volumeMin=1}}
 {id=Modulator_1, children [Random_8], properties {pitchMax=1.15, volumeMax=1, volumeMin=1}}
 {id=Modulator_4, children [WavePlayer_1], properties {volumeMax=1, volumeMin=0.15}}
 {id=Modulator_5, children [Random_1]}
 {id=Random_1, children [WavePlayer_6,WavePlayer_7,WavePlayer_8], properties {weights [1,1,1]}}
 {id=Random_6, children [WavePlayer_20, WavePlayer_21, WavePlayer_22], properties {weights [1,1,1]}}
 {id=Random_8, children [WavePlayer_23, WavePlayer_24, WavePlayer_25, WavePlayer_30], properties {weights [1,1,1,1]}}
 {
  id=SoundClass_0
  children [Attenuation_0]
  properties {soundClassOverride=&Fort_WeaponRanged_Fire_Distant}
 }
 {
  id=SoundClass_32
  children [Attenuation_49]
  properties {soundClassOverride=&Fort_WeaponRanged_Fire_3P_PSM}
 }
 {
  id=SoundClass_35
  children [Attenuation_34]
  properties {soundClassOverride=&Fort_WeaponRanged_Fire_Distant}
 }
 {id=WavePlayer_1, properties {soundWaveAssetPtr=&/Assault_Bone_Fire_Close_Chuff_01}}
 {id=WavePlayer_20, properties {soundWaveAssetPtr=&/Assault_Bone_Fire_Mid_03}}
 {id=WavePlayer_21, properties {soundWaveAssetPtr=&/Assault_Bone_Fire_Mid_02}}
 {id=WavePlayer_22, properties {soundWaveAssetPtr=&/Assault_Bone_Fire_Mid_01}}
 {id=WavePlayer_23, properties {soundWaveAssetPtr=&/Assault_Bone_Fire_Distant_01}}
 {id=WavePlayer_24, properties {soundWaveAssetPtr=&/Assault_Bone_Fire_Distant_02}}
 {id=WavePlayer_25, properties {soundWaveAssetPtr=&/Assault_Bone_Fire_Distant_03}}
 {id=WavePlayer_30, properties {soundWaveAssetPtr=&/Assault_Bone_Fire_Distant_04}}
 {id=WavePlayer_6, properties {soundWaveAssetPtr=&/Assault_Bone_Fire_Close_01}}
 {id=WavePlayer_7, properties {soundWaveAssetPtr=&/Assault_Bone_Fire_Close_02}}
 {id=WavePlayer_8, properties {soundWaveAssetPtr=&/Assault_Bone_Fire_Close_03}}
]
properties {
 maxDistance=27500
 volumeMultiplier=0.9
 attenuationOverrides {distanceAlgorithm=NaturalSound, falloffDistance=5000, attenuationShapeExtents {x=1000}}
 concurrencySet [&Weapon_Gunfire_Local_Conc, &Weapon_Gunfire_Global_Conc]
}
