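# Sound cue for the low-ammo firing sound of the Bone Assault Rifle.
# A single fire wave runs through a Modulator that adds a slight random upward pitch shift at constant volume, then into an Attenuation node using the shared close weapon projectile settings.
# The cue routes to the degraded low-ammo ranged weapon sound class and overrides concurrency on the cue itself.
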SoundCue /PrimalGameplay/Items/Guns/Bone/BoneAssaultRifle/Assault_Bone_Low_Ammo_Cue
root=Attenuation_49
nodes [
 {id=Attenuation_49, children [Modulator_3], properties {attenuationSettings=&Fort_Weapon_Projectile_Close_Athena_Att}}
 {id=Modulator_3, children [WavePlayer_1], properties {pitchMax=1.1, pitchMin=1, volumeMax=1, volumeMin=1}}
 {id=WavePlayer_1, properties {soundWaveAssetPtr=&/Assault_Bone_Fire_Low_Ammo}}
]
properties {
 bOverrideConcurrency=true
 maxDistance=4950
 soundClassObject=/Game/Sounds/FortSoundClassesAndModes/Fort_WeaponRanged_LowAmmo_Degraded_PSM
 volumeMultiplier=1
 attenuationOverrides {distanceAlgorithm=NaturalSound, falloffDistance=5000, attenuationShapeExtents {x=1000}}
}
