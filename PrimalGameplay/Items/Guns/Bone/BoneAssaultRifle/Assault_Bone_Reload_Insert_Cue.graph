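# Sound cue for the magazine-insert step of the Bone Assault Rifle reload.
# A single wave player feeds a default Modulator at the root, playing one insert sample at reduced volume.
# It routes through the weapon foley sound class with the shared weapon reload attenuation.
# A concurrency override prevents new instances from playing once the limit is reached.
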SoundCue /PrimalGameplay/Items/Guns/Bone/BoneAssaultRifle/Assault_Bone_Reload_Insert_Cue
root=Modulator_0
nodes [
 {id=Modulator_0, children [WavePlayer_1]}
 {id=WavePlayer_1, properties {soundWaveAssetPtr=&/Assault_Bone_Reload_Insert_01}}
]
properties {
 attenuationSettings=&Fort_Weapon_Reload_Att
 bOverrideConcurrency=true
 maxDistance=2250
 soundClassObject=&Fort_WeaponFoley
 volumeMultiplier=0.401
 attenuationOverrides {falloffDistance=4100, attenuationShapeExtents {x=900}}
 concurrencyOverrides {resolutionRule=PreventNew}
}
